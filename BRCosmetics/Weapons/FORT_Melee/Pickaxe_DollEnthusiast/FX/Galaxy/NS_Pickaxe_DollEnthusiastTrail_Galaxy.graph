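# Niagara swing trail for the Galaxy version of the DollEnthusiast pickaxe, built from a single ribbon emitter, Main_Trails.
# AnimTrail modules spawn, initialize and update short-lived particles from the skeletal mesh, feeding four ribbon renderers.
# System.Variant drives switches for trail color, brightness and a separate MotionTrailColor attribute, with one variant using an HDR color curve.
# Time-of-day inputs adjust color and intensity, while curves fade alpha.
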NiagaraSystem /BRCosmetics/Weapons/FORT_Melee/Pickaxe_DollEnthusiast/FX/Galaxy/NS_Pickaxe_DollEnthusiastTrail_Galaxy
dependencies {
 usedDataInterfaces [&NiagaraDataInterfaceColorCurve, &NiagaraDataInterfaceCurve]
 usedDynamicInputs []
 usedModules [
  &self:SystemSpawnScript.SystemScriptSource.SystemScriptGraph.NiagaraNodeAssignment_2.SetVariables_#1
  &self:SystemSpawnScript.SystemScriptSource.SystemScriptGraph.NiagaraNodeAssignment_0.SetVariables_#2
  &SystemState
 ]
}
emitters {Main_Trails {
  emitterUpdateScript [
   &EmitterState(Life Cycle Mode=::NewEnumerator1, Loop Behavior=::NewEnumerator1, Loop Delay=0.05, Loop Duration=0.2, UseLoopDelay=false)
   &FN_SpawnAnimTrail_v2()
  ]
  particleSpawnScript [
   /Game/Effects/Niagara/Modules/AnimTrail/FN_InitializeAnimTrail/FN_InitializeAnimTrailBoneIndecies()
   &FN_InitializeAnimTrail(
    AnimTrailWidth=32
    Color=&MergeHelper_ColorSwitch(
     0=&FNTimeOfDayColor(
      Night Color=<rgba 0.306 0.0683 0.6038 1>
      Morning Color=<rgba 0.6302 0.1805 0.3666 1>
      Day Color=<rgba 0.6302 0.1805 0.3666 1>
      Evening Color=<rgba 0.6302 0.1805 0.3666 1>
     )
     1=<rgba 1 0 0.1475 1>
     2=<rgba 1 0.4909 0.1422 1>
     3=<rgba 0.5312 0.931 2 0.25>
     4=<rgba 0.3095 0.6095 1 1>
     5=<rgba 0.6055 0.0916 1 1>
     Index=$System.Variant
     Num=::NewEnumerator4
    )
    Lifetime=0.1
    SkeletalMesh=$System.Skeletal Mesh
   ) {module=FN_InitializeAnimTrail001}
  ]
  particleUpdateScript [
   &FN_UpdateAnimTrail(SkeletalMesh=$System.Skeletal Mesh)
   &ScaleColor(
    Linear Color Curve=DataInterface() {propertyValues {
      AlphaCurve {keys [<at 0 1>,<at 1 0>]}
      BlueCurve {keys [<at 0 0.4792>,<at 0.2078 0.3968>]}
      GreenCurve {keys [<at 0 0.197>,<at 0.2078 0.1559>]}
      RedCurve {keys [<at 0 0.0309>,<at 0.2078 0.7379>]}
     }}
    Scale Mode=::NewEnumerator2
   ) {enabled=false, module=ScaleColor002}
   &Color(Color=&MergeHelper_ColorSwitch(
     0=$StackContext.Initial.Color
     1=$StackContext.Initial.Color
     2=$StackContext.Initial.Color
     3=&ColorFromCurve(ColorCurve=DataInterface() {propertyValues {
        AlphaCurve {keys [<at 0 0.125 RCIM_Cubic>]}
        BlueCurve {keys [<at 0 5 RCIM_Cubic>,<at 0.261 5>,<at 0.5912 5 RCIM_Cubic>]}
        GreenCurve {keys [<at 0 1.7726 RCIM_Cubic>,<at 0.261 1.7726>,<at 0.5912 0.6 RCIM_Cubic>]}
        RedCurve {keys [<at 0 0.355 RCIM_Cubic>,<at 0.261 0.355>,<at 0.5912 4.1952 RCIM_Cubic>]}
       }})
     4=$StackContext.Initial.Color
     Index=$System.Variant
     Num=::NewEnumerator3
    ))
   /SetVariables(Particles.MotionTrailColor:LinearColor=&MultiplyLinearColor(
     Linear Color=&MergeHelper_ColorSwitch(
      0=$Particles.Color
      1=$Particles.Color
      2=<rgba 0.035 0.035 0.035 0.5>
      3=$Particles.Color
      4=$Particles.Color
      5=<rgba 0.0868 0.609 1 1>
      Index=$System.Variant
      Num=::NewEnumerator4
     )
     Scale Factor=&MakeVector4(
      X=&FNTimeOfDayScalar(Night Scale=0.4, Morning Scale=0.5, Day Scale=0.5, Evening Scale=0.5)
      Y=&FNTimeOfDayScalar(Night Scale=0.4, Morning Scale=0.5, Day Scale=0.5, Evening Scale=0.5)
      Z=&FNTimeOfDayScalar(Night Scale=0.4, Morning Scale=0.5, Day Scale=0.5, Evening Scale=0.5)
      W=&FloatFromCurve(
       FloatCurve=DataInterface() {propertyValues {Curve {keys [
           {arriveTangent=-6.7288, interpMode=RCIM_Cubic, leaveTangent=-6.7288, tangentMode=RCTM_User, time=0, value=1}
           {tangentMode=RCTM_User, time=0.35, value=0}
          ]}}}
       Scale Curve=0.25
      )
     )
    )) {isSetParameters=true}
   /SetVariables(Particles.GalaxyWidth=30) {isSetParameters=true}
   &ScaleColor(
    Color Value To Scale=$StackContext.Color
    Scale Alpha=&FloatFromCurve(
     FloatCurve=DataInterface() {propertyValues {Curve {keys [{interpMode=RCIM_Cubic, tangentMode=RCTM_User, time=0, value=1}, {interpMode=RCIM_Cubic, tangentMode=RCTM_User, time=1, value=0}]}}}
     Scale Curve=0.75
    )
    Scale RGB=&VectorFromFloat(Value=&MergeHelper_FloatSwitch(
      0=&FN_Select_Float_from_Time_Of_Day(Morning=1, Day=0.65, Evening=1, Night=0.75)
      1=&FN_Select_Float_from_Time_Of_Day(Morning=1, Day=0.65, Evening=1, Night=0.75)
      2=&FN_Select_Float_from_Time_Of_Day(Morning=1, Day=0.65, Evening=1, Night=0.75)
      3=1
      4=&FN_Select_Float_from_Time_Of_Day(Morning=1, Day=0.65, Evening=1, Night=0.75)
      Index=$System.Variant
      Num=::NewEnumerator3
     ))
    ScaleA=true
    ScaleRGB=true
   )
  ]
  rendererClasses [&NiagaraRibbonRendererProperties]
  renderers [
   {rendererClass=&NiagaraRibbonRendererProperties}
   {rendererClass=&NiagaraRibbonRendererProperties, rendererIndex=1}
   {rendererClass=&NiagaraRibbonRendererProperties, rendererIndex=2}
   {rendererClass=&NiagaraRibbonRendererProperties, rendererIndex=3}
  ]
 }}
summary {
 systemName=NS_Pickaxe_DollEnthusiastTrail_Galaxy
 emitters [{emitterName=Main_Trails, rendererClasses [&NiagaraRibbonRendererProperties]}]
 userVariables [
  User.RibbonIDOffset {type=NiagaraInt32}
  User.SkeletalMesh {
   type=NiagaraDataInterfaceSkeletalMesh
   defaultValue {
    struct=NiagaraExt_VariableValue_DataInterface
    value {dataInterface=&self:NiagaraDataInterfaceSkeletalMesh_0}
   }
  }
 ]
}
